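# Material function for crew-style characters that applies an optional ice look to incoming material attributes.
# When the Ice Style static switch is on, MF_Silk_TechArt processes the material using the TechArt input, then the shared weapon-skin Ice function layers on top.
# Channel mask parameters pick ice coverage and edge map from TechArt channels.
# With the switch off, the material passes through untouched; the Fresnel input is unconnected.
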MaterialFunction /Game/Global/Materials/Character_Parent_2023/CrewStyle_Parent_2023/Ice/MF_CrewStyle_Ice
graph {nodes [
  FunctionInput() {inputName=Fresnel, inputType=FunctionInput_Scalar}
  n1 = FunctionInput() {inputType=FunctionInput_MaterialAttributes}
  n2 = FunctionInput() {inputName=TechArt}
  n3 = MaterialFunctionCall!(Material=@n1, Alpha=@n2) {materialFunction=/Game/Global/MaterialFunctions/TechArt/MF_Silk_TechArt}
  n4 = ChannelMaskParameter(None=@n2) {maskChannel=Green, parameterName=IceEdgeChannel, defaultValue=<rgba 0 1 0 0>}
  n5 = ChannelMaskParameter(None=@n2) {parameterName=IceChannel}
  FunctionOutput(None=StaticSwitchParameter(
    True=MaterialFunctionCall!(Material=@n3.Result, EdgeMap=@n4.R, Alpha=@n5.R) {materialFunction=/Game/Weapons/WeaponSkins/MaterialFunction/Ice}.Result
    False=@n1
   ) {parameterName=Ice Style})
 ]}
